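# Sound cue for an enemy portal spawn, built from husk spawn effect waves and routed to the enemy spawn sound class.
# A quality-level node picks between an equal-weight random set of nine waves, a smaller random set of three, or one of two single-wave players.
# Playback is at reduced volume, with husk spawn attenuation and a concurrency set named for two instances per owner, stopping the oldest.
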SoundCue /Game/Sounds/Fort_GamePlay_Sounds/Portal/EnemyPortal_Spawn_Cue
root=QualityLevel_0
nodes [
 {id=QualityLevel_0, children [Random_1,Random_2,WavePlayer_0,WavePlayer_2]}
 {
  id=Random_1
  children [WavePlayer_9, WavePlayer_10, WavePlayer_11, WavePlayer_12, WavePlayer_13, WavePlayer_14, WavePlayer_15, WavePlayer_16, WavePlayer_17]
  properties {weights [1,1,1,1,1,1,1,1,1]}
 }
 {id=Random_2, children [WavePlayer_18, WavePlayer_20, WavePlayer_25], properties {weights [1,1,1]}}
 {id=WavePlayer_0, properties {soundWaveAssetPtr=&/husk_spawn_neweffect_01b}}
 {id=WavePlayer_10, properties {soundWaveAssetPtr=&/husk_spawn_neweffect_02b}}
 {id=WavePlayer_11, properties {soundWaveAssetPtr=&/husk_spawn_neweffect_03b}}
 {id=WavePlayer_12, properties {soundWaveAssetPtr=&/husk_spawn_neweffect_04b}}
 {id=WavePlayer_13, properties {soundWaveAssetPtr=&/husk_spawn_neweffect_05b}}
 {id=WavePlayer_14, properties {soundWaveAssetPtr=&/husk_spawn_neweffect_06b}}
 {id=WavePlayer_15, properties {soundWaveAssetPtr=&/husk_spawn_neweffect_08b}}
 {id=WavePlayer_16, properties {soundWaveAssetPtr=&/husk_spawn_neweffect_11b}}
 {id=WavePlayer_17, properties {soundWaveAssetPtr=&/husk_spawn_neweffect_10b}}
 {id=WavePlayer_18, properties {soundWaveAssetPtr=&husk_spawn_neweffect_01b}}
 {id=WavePlayer_2, properties {soundWaveAssetPtr=&husk_spawn_neweffect_01b}}
 {id=WavePlayer_20, properties {soundWaveAssetPtr=&husk_spawn_neweffect_03b}}
 {id=WavePlayer_25, properties {soundWaveAssetPtr=&husk_spawn_neweffect_11b}}
 {id=WavePlayer_9, properties {soundWaveAssetPtr=&husk_spawn_neweffect_01b}}
]
properties {
 attenuationSettings=/Game/Sounds/Attenuation/Fort_Monster_Husk_Spawn_Att
 maxDistance=9500
 soundClassObject=/Game/Sounds/FortSoundClassesAndModes/Fort_Enemy_Spawn
 volumeMultiplier=0.5
 attenuationOverrides {falloffDistance=4000, attenuationShapeExtents {x=1000}}
 concurrencyOverrides {maxCount=2, resolutionRule=StopFarthestThenPreventNew}
 concurrencySet [/Game/Sounds/Concurrency_Rules/TwoMaximum_Owner_StopOldest]
}
